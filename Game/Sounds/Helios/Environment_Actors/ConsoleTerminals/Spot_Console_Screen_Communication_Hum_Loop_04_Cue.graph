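# Quiet looping sound cue for the communication hum of a console terminal screen, classed as an ambient prop sound.
# One looping wave player feeds a Modulator that randomizes playback, with the minimum pitch raised to unity.
# Spatialization comes from a borrowed Asteria small-screen static capsule attenuation asset.
# A local concurrency override blocks new instances at the limit, superseding the assigned weapon-drop stop-oldest concurrency asset.
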SoundCue /Game/Sounds/Helios/Environment_Actors/ConsoleTerminals/Spot_Console_Screen_Communication_Hum_Loop_04_Cue
root=Modulator_0
nodes [
 {id=Modulator_0, children [WavePlayer_0], properties {pitchMin=1}}
 {
  id=WavePlayer_0
  properties {bLooping=true, soundWaveAssetPtr=&/Spot_Console_Terminal_Communication_Hum_Loop_04}
 }
]
properties {
 attenuationSettings=/Game/Sounds/Asteria/Environment_Actors/Screen_Small/Spot_Small_Screen_Static_Capsule_Att
 bOverrideConcurrency=true
 maxDistance=854.7964
 soundClassObject=&Fort_Ambient_Props_Athena
 volumeMultiplier=0.1
 attenuationOverrides {falloffDistance=2000}
 concurrencyOverrides {resolutionRule=PreventNew}
 concurrencySet [&Weapon_Drop_StopOldest]
}
