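# Sound cue for a looping communication hum spot ambience on Helios console terminals, classed as an Athena ambient prop sound.
# A quality-level node routes all four of its inputs to the same looping wave player, so the hum is identical at every quality.
# A shared hum attenuation asset handles spatialization, and volume is reduced.
# Overridden concurrency caps simultaneous instances per owner.
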SoundCue /Game/Sounds/Helios/Environment_Actors/ConsoleTerminals/Spot_Console_Terminal_Communication_Hum_Loop_03_Cue
root=QualityLevel_0
nodes [
 {id=QualityLevel_0, children [WavePlayer_9,WavePlayer_9,WavePlayer_9,WavePlayer_9]}
 {
  id=WavePlayer_9
  properties {bLooping=true, soundWaveAssetPtr=&/Spot_Console_Terminal_Communication_Hum_Loop_03}
 }
]
properties {
 attenuationSettings=&/Spot_Console_Terminal_Communication_Hum_Att
 bOverrideConcurrency=true
 maxDistance=1000
 soundClassObject=&Fort_Ambient_Props_Athena
 subtitlePriority=5000
 volumeMultiplier=0.501
 attenuationOverrides {
  bEnableFocusInterpolation=true
  bEnableListenerFocus=true
  bEnableOcclusion=true
  dBAttenuationAtMax=-40
  falloffDistance=512
  focusAzimuth=15
  focusVolumeAttenuation=1.5
  nonFocusAzimuth=45
  nonFocusVolumeAttenuation=0.5
  occlusionInterpolationTime=1
  occlusionLowPassFilterFrequency=10000
  occlusionTraceChannel=ECC_GameTraceChannel11
  occlusionVolumeAttenuation=0.5
  stereoSpread=512
  attenuationShapeExtents=<256 64.5942 90>
 }
 concurrencyOverrides {bLimitToOwner=true, maxCount=2}
}
